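# Idle Niagara cosmetic effect for the Ebony character body, drawing flickering white-violet lightning at the eyes and shoulders.
# Eyes_FX and Shoulders_FX burst mirrored mesh bolts pinned to the skeletal mesh, gated by user amount parameters.
# A random per-bolt flag picks traveling or stationary animation through dynamic material parameters, with flicker and time-of-day brightness scaling.
# Sparks_Eyes spawns from young eye particles, using their Mass-encoded side to aim falling sparks.
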NiagaraSystem /BRCosmetics/Characters/Player/Male/Medium/Bodies/M_MED_Ebony/FX/NS_Ebony_White_Idle
dependencies {
 usedDataInterfaces [
  &NiagaraDataInterfaceArrayFloat
  &NiagaraDataInterfaceArrayFloat3
  &NiagaraDataInterfaceCurve
  &NiagaraDataInterfaceColorCurve
  &NiagaraDataInterfaceParticleRead
  &NiagaraDataInterfaceVector2DCurve
 ]
 usedDynamicInputs [&Boolean_NotOperation]
 usedModules [
  &self:SystemSpawnScript.SystemScriptSource.SystemScriptGraph.NiagaraNodeAssignment_1.SetVariables_#1
  &SystemState
  &FN_Timer
  &self:SystemSpawnScript.SystemScriptSource.SystemScriptGraph.NiagaraNodeAssignment_0.SetVariables_#2
  &self:SystemSpawnScript.SystemScriptSource.SystemScriptGraph.NiagaraNodeAssignment_2.SetVariables_#3
 ]
}
emitters {
 Eyes_FX {
  emitterUpdateScript [
   &EmitterState()
   &SpawnBurst_Instantaneous(
    Spawn Count=&RandomRangeInt(Minimum=1, Maximum=2)
    Spawn Probability=&Multiply_Float(A=$User.MaskParticlesAmount, B=$User.ParticlesAmount)
    Spawn Time=&RandomRangeFloat(Minimum=0, Maximum=0.5)
    Use Spawn Probability=true
    UseLoopCountLimit=false
   )
   /SetVariables(Emitter.DepthFade=35) {isSetParameters=true, module=SetVariables_#4}
  ]
  particleSpawnScript [
   &FN_SetSpawnGroup(Number of Spawn Groups=2)
   /SetVariables(
    Particles.MeshIndex=&RandomRangeInt(Minimum=0, Maximum=2)
    Particles.bIsTravelingBolt:NiagaraBool=&SetBoolByFloatComparison(A=&RandomRangeFloat(Minimum=0, Maximum=1), B=0.3)
   ) {isSetParameters=true, module=SetVariables_#5}
   /SetVariables(Particles.MeshIndex=1) {enabled=false, isSetParameters=true, module=SetVariables_#6}
   &InitializeParticle(
    Color=<rgba 158.3412 154 220 1>
    Lifetime=&MakeCustomFloatFromBool(
     Bool=$Particles.bIsTravelingBolt
     True Float=&RandomRangeFloat(Minimum=0.42, Maximum=0.38)
     False Float=&RandomRangeFloat(Minimum=0.3, Maximum=0.2)
    )
    Mesh Scale=&Multiply_VectorByFloat(
     Vector=&SelectVectorFromArray(Direct Array Index=$Particles.SpawnGroup, Vector Selection Array=DataInterface() {propertyValues {FloatData [<1 1 -1>, <1 1 1>]}})
     Float=&RandomRangeFloat(Minimum=0.8, Maximum=1.25)
    )
    Mesh Scale Mode=::NewEnumerator4
    Sprite Rotation Mode=::NewEnumerator2
    Sprite Size Mode=::NewEnumerator3
    Uniform Sprite Size=&Multiply_Float(
     A=&RandomRangeFloat(Minimum=13, Maximum=17)
     B=&SelectFloatFromArray(
      Direct Array Index=$Particles.MeshIndex
      Float Selection Array=DataInterface() {propertyValues {FloatData [1,1,1,1,0,0,1,1,1,1,1,1,1,1,1,1,1,1]}}
     )
    )
    UsePositionOffset=false
    Write Material Random=false
   )
   &DynamicMaterialParameters(
    Index 0 Param 1=$System.LightningWidth
    Index 0 Param 2=&MakeCustomFloatFromBool(Bool=$Particles.bIsTravelingBolt, True Float=0.75, False Float=3)
    Index 0 Param 4=0.5
    Module.Param0WriteEnabled=true
    Module.Param1WriteEnabled=true
    Module.Param2WriteEnabled=true
    Module.Param3WriteEnabled=true
   )
   /SetVariables() {isSetParameters=true, module=SetVariables_#7}
   /SetVariables(Particles.GlowGlowColorInit:LinearColor=<rgba 0.0715 0.07 0.1 0.5>, Particles.GlowSizeInit:Vector2f=<25 25>) {isSetParameters=true, module=SetVariables_#8}
   &CameraOffset(Camera Offset Amount=1)
   /SetVariables(Particles.Mass:NiagaraFloat=&MakeFloatFromInt(Int=$Particles.SpawnGroup)) {isSetParameters=true}
  ]
  particleUpdateScript [
   &SolveForcesAndVelocity(Clamp Velocity=false, Limit Acceleration=false)
   &SkeletalMeshLocation(
    Bone Sampling Mode=::NewEnumerator4
    MaskSpawnGroup=false
    Orientation Sampling=::NewEnumerator1
    Sampled Position Offset=&SelectVectorFromArray(Direct Array Index=$Particles.SpawnGroup, Vector Selection Array=DataInterface() {propertyValues {FloatData [<5 9.5 4>, <5 9.5 -4>]}})
    Skeletal Mesh=$User.SkeletalMesh
   )
   /SetVariables(Particles.MeshPosition:NiagaraPosition=$Particles.Position) {isSetParameters=true, module=SetVariables_#9}
   &SkeletalMeshLocation(
    Bone Sampling Mode=::NewEnumerator4
    MaskSpawnGroup=false
    Sampled Position Offset=&SelectVectorFromArray(Direct Array Index=$Particles.SpawnGroup, Vector Selection Array=DataInterface() {propertyValues {FloatData [<5 9.5 4>, <5 9.5 -4>]}})
    Skeletal Mesh=$User.SkeletalMesh
   ) {module=SkeletalMeshLocation001}
   /SetVariables(Particles.FlarePosition:NiagaraPosition=$Particles.Position) {isSetParameters=true}
   &ParticleState()
   &DynamicMaterialParameters(
    Index 0 Param 1=&MakeCustomFloatFromBool(
     Bool=$Particles.bIsTravelingBolt
     True Float=$System.LightningWidth
     False Float=&FloatFromCurve(FloatCurve=DataInterface() {propertyValues {Curve {keys [<at 0.666 1>,<at 1 0>]}}}, Scale Curve=$System.LightningWidth)
    )
    Index 0 Param 3=&MakeCustomFloatFromBool(
     Bool=$Particles.bIsTravelingBolt
     True Float=$Particles.NormalizedAge
     False Float=&FloatFromCurve(FloatCurve=DataInterface() {propertyValues {Curve {keys [<at 0 0>,<at 0.333 0.5>]}}})
    )
    Index 0 Param 4=&MakeCustomFloatFromBool(
     Bool=$Particles.bIsTravelingBolt
     True Float=0.5
     False Float=&FloatFromCurve(FloatCurve=DataInterface() {propertyValues {Curve {keys [{tangentMode=RCTM_User, time=0, value=0.333}, {tangentMode=RCTM_User, time=1, value=0.666}]}}})
    )
    Module.Param0WriteEnabled=true
    Module.Param1WriteEnabled=false
    Module.Param2WriteEnabled=true
    Module.Param3WriteEnabled=true
   ) {module=DynamicMaterialParameters001}
   &ScaleSpriteSize(Uniform Curve Sprite Scale=DataInterface() {propertyValues {Curve {keys [
        {arriveTangent=3.7119, interpMode=RCIM_Cubic, leaveTangent=3.7119, tangentMode=RCTM_User, time=0.3, value=0.25}
        {interpMode=RCIM_Cubic, tangentMode=RCTM_User, time=0.5, value=1}
       ]}}})
   &ScaleColor(
    Scale RGB=&VectorFromFloat(Value=&Multiply_Float(A=&Lerp_Float(A=0.133, B=1, Alpha=$System.Flicker), B=1))
    ScaleA=false
    ScaleRGB=true
   )
   /SetVariables(Particles.FlareColor=&ColorFromCurve(ColorCurve=DataInterface() {propertyValues {
       AlphaCurve {keys [<at 0 1 RCIM_Cubic>,<at 0.7365 0 RCIM_Cubic>]}
       BlueCurve {keys [<at 0 33 RCIM_Cubic>,<at 0.274 11>,<at 1 1 RCIM_Cubic>]}
       GreenCurve {keys [<at 0 23.1 RCIM_Cubic>,<at 0.274 7.7>,<at 1 0.7 RCIM_Cubic>]}
       RedCurve {keys [<at 0 23.595 RCIM_Cubic>,<at 0.274 7.865>,<at 1 0.715 RCIM_Cubic>]}
      }})) {isSetParameters=true, module=SetVariables_#10}
   /SetVariables(
    Particles.FlareColor=&MultiplyLinearColor(
     Linear Color=$Particles.FlareColor
     Scale Factor=&MakeFloat4FromVectorAndFloat("Vector (RGB)"=&VectorFromFloat(Value=0.2), "Float (Alpha)"=1)
    )
    Particles.GlowColor=&MakeLinearColorFromVectorAndFloat(
     "Vector (RGB)"=&MakeVectorFromLinearColorRGB(Color=$Particles.GlowGlowColorInit)
     "Float (Alpha)"=&FloatFromCurve(
      FloatCurve=DataInterface() {propertyValues {Curve {keys [<at 0 0.5>, {interpMode=RCIM_Cubic, tangentMode=RCTM_User, time=0.3, value=1}, {tangentMode=RCTM_User, time=0.75, value=0}]}}}
      Scale Curve=&FNTimeOfDayScalar(Night Scale=1, Morning Scale=1.2, Day Scale=1.5, Evening Scale=1.2)
     )
    )
    Particles.GlowSize=&Multiply_Vector2D_ByFloat(
     Float=&FloatFromCurve(
      FloatCurve=DataInterface() {propertyValues {Curve {keys [{arriveTangent=0.3811, interpMode=RCIM_Cubic, leaveTangent=0.3811, tangentMode=RCTM_User, time=0.3, value=1}, <at 0.5 0>]}}}
      Scale Curve=1.2
     )
     Vector2D=$Particles.GlowSizeInit
    )
   ) {isSetParameters=true, module=SetVariables_#11}
   &ScaleColor(
    Color Value To Scale=$Particles.Color
    Scale RGB=&Multiply_VectorByFloat(Vector=<1 1 1>, Float=$System.TimeOfDayBrightnessMultiplier)
    ScaleA=true
    ScaleRGB=true
   ) {module=ScaleColor001}
   &GenerateDeathEvent(Use Event Probability=false)
  ]
  rendererClasses [&NiagaraMeshRendererProperties, &NiagaraSpriteRendererProperties]
  renderers [
   {rendererClass=&NiagaraMeshRendererProperties, rendererIndex=0}
   {rendererClass=&NiagaraSpriteRendererProperties, rendererIndex=1}
   {rendererClass=&NiagaraSpriteRendererProperties, rendererIndex=2}
  ]
 }
 Shoulders_FX {
  emitterUpdateScript [
   &EmitterState(
    Inactive Response=::NewEnumerator2
    Life Cycle Mode=::NewEnumerator1
    Loop Duration=&RandomRangeFloat(Minimum=0.1, Maximum=0.8)
    Recalculate Duration Each Loop=true
    UseLoopDelay=false
   )
   &SpawnBurst_Instantaneous(
    Spawn Count=&RandomRangeInt(Minimum=1, Maximum=3)
    Spawn Probability=$User.ParticlesAmount
    Spawn Time=&RandomRangeFloat(Minimum=0, Maximum=0.1)
    Use Spawn Probability=true
    UseLoopCountLimit=false
   )
   /SetVariables(Emitter.DepthFade=35) {isSetParameters=true, module=SetVariables_#4}
  ]
  particleSpawnScript [
   &FN_SetSpawnGroup(Number of Spawn Groups=2)
   /SetVariables(
    Particles.MeshIndex=&RandomRangeInt(Minimum=0, Maximum=5)
    Particles.bIsTravelingBolt:NiagaraBool=&SetBoolByFloatComparison(
     A=&RandomRangeFloat(Minimum=0, Maximum=1)
     B=&FN_FrontEndFloatModifier(Gameplay=0.12, Front End=0)
    )
   ) {isSetParameters=true, module=SetVariables_#5}
   /SetVariables(Particles.MeshIndex=4) {enabled=false, isSetParameters=true, module=SetVariables_#6}
   &InitializeParticle(
    Color=<rgba 158.3412 154 220 1>
    Lifetime=&MakeCustomFloatFromBool(
     Bool=$Particles.bIsTravelingBolt
     True Float=&RandomRangeFloat(Minimum=0.42, Maximum=0.38)
     False Float=&RandomRangeFloat(Minimum=0.33, Maximum=0.22)
    )
    Mesh Scale=&Multiply_VectorByFloat(
     Vector=&SelectVectorFromArray(Direct Array Index=$Particles.SpawnGroup, Vector Selection Array=DataInterface() {propertyValues {FloatData [<1 1 1>, <1 -1 1>]}})
     Float=&RandomRangeFloat(Minimum=1, Maximum=1)
    )
    Mesh Scale Mode=::NewEnumerator4
    Sprite Rotation Mode=::NewEnumerator2
    Sprite Size Mode=::NewEnumerator3
    Uniform Sprite Size=&Multiply_Float(
     A=&RandomRangeFloat(Minimum=9, Maximum=17)
     B=&SelectFloatFromArray(
      Direct Array Index=$Particles.MeshIndex
      Float Selection Array=DataInterface() {propertyValues {FloatData [1,1,1,1,0,0,1,1,1,1,1,1,1,1,1,1,1,1]}}
     )
    )
    UsePositionOffset=true
    Write Material Random=false
   )
   &DynamicMaterialParameters(
    Index 0 Param 1=$System.LightningWidth
    Index 0 Param 2=&MakeCustomFloatFromBool(
     Bool=$Particles.bIsTravelingBolt
     True Float=&RandomRangeFloat(Minimum=0.5, Maximum=0.75)
     False Float=2
    )
    Index 0 Param 4=0.5
    Module.Param0WriteEnabled=true
    Module.Param1WriteEnabled=true
    Module.Param2WriteEnabled=true
    Module.Param3WriteEnabled=true
   )
   /SetVariables(Particles.FlareCameraOffset=1) {isSetParameters=true, module=SetVariables_#7}
   /SetVariables(Particles.GlowGlowColorInit:LinearColor=<rgba 0.0715 0.07 0.1 0.5>, Particles.GlowSizeInit:Vector2f=<15 15>) {isSetParameters=true, module=SetVariables_#8}
   &CameraOffset(Camera Offset Amount=2)
  ]
  particleUpdateScript [
   &SolveForcesAndVelocity(Clamp Velocity=false, Limit Acceleration=false)
   &SkeletalMeshLocation(
    Bone / Socket Index=$Particles.SpawnGroup
    MaskSpawnGroup=false
    Mesh Sampling Type=::NewEnumerator5
    Sampled Position Offset=&SelectVectorFromArray(
     Direct Array Index=$Particles.MeshIndex
     Vector Selection Array=DataInterface() {propertyValues {FloatData [<0 0 0>, <0 0 0>, <0 0 0>, <0 0 0>, <0 0 0>]}}
    )
    Skeletal Mesh=$User.SkeletalMesh
    Socket Sampling Mode=::NewEnumerator5
   )
   /SetVariables(Particles.MeshPosition:NiagaraPosition=$Particles.Position) {isSetParameters=true, module=SetVariables_#9}
   &ParticleState()
   &DynamicMaterialParameters(
    Index 0 Param 1=&MakeCustomFloatFromBool(
     Bool=$Particles.bIsTravelingBolt
     True Float=$System.LightningWidth
     False Float=&FloatFromCurve(FloatCurve=DataInterface() {propertyValues {Curve {keys [<at 0.666 1>,<at 1 0>]}}}, Scale Curve=$System.LightningWidth)
    )
    Index 0 Param 3=&MakeCustomFloatFromBool(
     Bool=$Particles.bIsTravelingBolt
     True Float=$Particles.NormalizedAge
     False Float=&FloatFromCurve(FloatCurve=DataInterface() {propertyValues {Curve {keys [<at 0 0>,<at 0.333 0.5>]}}})
    )
    Index 0 Param 4=&MakeCustomFloatFromBool(
     Bool=$Particles.bIsTravelingBolt
     True Float=0.5
     False Float=&FloatFromCurve(FloatCurve=DataInterface() {propertyValues {Curve {keys [{tangentMode=RCTM_User, time=0, value=0.333}, {tangentMode=RCTM_User, time=1, value=0.666}]}}})
    )
    Module.Param0WriteEnabled=true
    Module.Param1WriteEnabled=false
    Module.Param2WriteEnabled=true
    Module.Param3WriteEnabled=true
   ) {module=DynamicMaterialParameters001}
   &ScaleColor(
    Scale RGB=&VectorFromFloat(Value=&Multiply_Float(A=&Lerp_Float(A=0.133, B=1, Alpha=$System.Flicker), B=1))
    ScaleA=false
    ScaleRGB=true
   )
   /SetVariables(Particles.FlareColor=&ColorFromCurve(ColorCurve=DataInterface() {propertyValues {
       AlphaCurve {keys [<at 0 1 RCIM_Cubic>,<at 0.7365 0 RCIM_Cubic>]}
       BlueCurve {keys [<at 0 33 RCIM_Cubic>,<at 0.274 11>,<at 1 1 RCIM_Cubic>]}
       GreenCurve {keys [<at 0 23.1 RCIM_Cubic>,<at 0.274 7.7>,<at 1 0.7 RCIM_Cubic>]}
       RedCurve {keys [<at 0 23.595 RCIM_Cubic>,<at 0.274 7.865>,<at 1 0.715 RCIM_Cubic>]}
      }})) {isSetParameters=true, module=SetVariables_#10}
   /SetVariables(
    Particles.FlareColor=&MultiplyLinearColor(
     Linear Color=$Particles.FlareColor
     Scale Factor=&MakeFloat4FromVectorAndFloat("Vector (RGB)"=&VectorFromFloat(Value=0.15), "Float (Alpha)"=$System.TimeOfDayBrightnessMultiplier)
    )
    Particles.GlowColor=&MakeLinearColorFromVectorAndFloat(
     "Vector (RGB)"=&MakeVectorFromLinearColorRGB(Color=$Particles.GlowGlowColorInit)
     "Float (Alpha)"=&FloatFromCurve(
      FloatCurve=DataInterface() {propertyValues {Curve {keys [<at 0 0.4>, {interpMode=RCIM_Cubic, tangentMode=RCTM_User, time=0.3, value=0.8}, {tangentMode=RCTM_User, time=0.75, value=0}]}}}
      Scale Curve=$System.TimeOfDayBrightnessMultiplier
     )
    )
    Particles.GlowSize=&Multiply_Vector2D_ByFloat(
     Float=&FloatFromCurve(FloatCurve=DataInterface() {propertyValues {Curve {keys [{arriveTangent=0.3811, interpMode=RCIM_Cubic, leaveTangent=0.3811, tangentMode=RCTM_User, time=0.3, value=1}, <at 0.5 0>]}}})
     Vector2D=$Particles.GlowSizeInit
    )
   ) {isSetParameters=true, module=SetVariables_#11}
   &ScaleColor(
    Color Value To Scale=$Particles.Color
    Scale RGB=&Multiply_VectorByFloat(Vector=<1 1 1>, Float=$System.TimeOfDayBrightnessMultiplier)
    ScaleA=true
    ScaleRGB=true
   ) {module=ScaleColor001}
  ]
  rendererClasses [&NiagaraMeshRendererProperties]
  renderers [{rendererClass=&NiagaraMeshRendererProperties, rendererIndex=0}]
 }
 Sparks_Eyes {
  emitterUpdateScript [
   &EmitterState()
   &SpawnParticlesFromOtherEmitter(
    Attribute Reader=DataInterface() {propertyValues {EmitterBinding {bindingMode=Other, emitterName=Eyes_FX}}}
    Maximum Age Spawn Threshold=0.1
    Spawn Rate=25
    Spawn Rate Per Particle Cap=500
    Use Maximum Age=true
    Use Minimum Age=false
   )
  ]
  particleSpawnScript [
   &InitializeParticle(
    Color=<rgba 6.435 6.3 9 1>
    Lifetime=&RandomRangeFloat(Minimum=0.066, Maximum=0.444)
    Sprite Size=&Lerp_Vector2(A=<1.2 0.5>, B=<0.7 3>, Alpha=&RandomRangeFloat(Minimum=0, Maximum=1))
    Sprite Size Mode=::NewEnumerator4
    UsePositionOffset=false
    Write Material Random=true
   )
   &SampleParticlesFromOtherEmitter(Mass Application:ENiagara_LifetimeInheritanceOptions=::NewEnumerator2, Mass Sampling:ENiagara_DirectReadSamplingMode=::NewEnumerator1)
   /SetVariables(Particles.SpawnGroup=&MakeIntFromFloat(Float=$Output.SampleParticlesFromOtherEmitter.SampledMass)) {isSetParameters=true}
   &AddVelocity(
    Cone Angle=33
    Cone Axis=&SelectVectorFromArray(Direct Array Index=$Particles.SpawnGroup, Vector Selection Array=DataInterface() {propertyValues {FloatData [<-1 0 0.5>, <1 0 0.5>]}})
    Module.UseConeFalloff=false
    Module.UseVelDistribution=false
    Override Seed=false
    Velocity Mode=::NewEnumerator2
    Velocity Speed=&RandomRangeFloat(Minimum=11, Maximum=111)
   )
   /SetVariables(Particles.SubImageIndex=&RandomRangeFloat(Minimum=0, Maximum=4)) {isSetParameters=true}
  ]
  particleUpdateScript [
   &Drag(Drag=&RandomRangeFloat(Minimum=3, Maximum=6), Use Linear Drag=true, Use Rotational Drag=false)
   &GravityForce(Gravity=<0 0 -111>)
   &SolveForcesAndVelocity(Clamp Velocity=false, Limit Acceleration=false)
   &ParticleState()
   &ScaleSpriteSize(
    Non-Uniform Curve Sprite Scale=DataInterface() {propertyValues {
      XCurve {keys [{tangentMode=RCTM_User, time=0, value=1}, <at 0.6 1>, {tangentMode=RCTM_User, time=1, value=0}]}
      YCurve {keys [{tangentMode=RCTM_User, time=0, value=1}, {tangentMode=RCTM_User, time=1, value=0}]}
     }}
    Scale Sprite Size Mode=::NewEnumerator3
   )
   &ScaleSpriteSizeBySpeed(Max Scale Factor=<1 3>, Velocity Threshold=66)
  ]
  rendererClasses [&NiagaraSpriteRendererProperties]
  renderers [{rendererClass=&NiagaraSpriteRendererProperties}]
 }
}
summary {
 systemName=NS_Ebony_White_Idle
 emitters [
  {emitterName=Eyes_FX, rendererClasses [&NiagaraMeshRendererProperties, &NiagaraSpriteRendererProperties]}
  {emitterName=Sparks_Eyes, rendererClasses [&NiagaraSpriteRendererProperties]}
  {emitterName=Shoulders_FX, rendererClasses [&NiagaraMeshRendererProperties]}
 ]
 userVariables [
  User.bIsFrontEnd {type=NiagaraBool}
  User.bIsFrontEndPreview {type=NiagaraBool}
  User.ColorSwitcher {type=NiagaraFloat, defaultValue=0, description=""}
  User.FlickerSpeed {type=NiagaraFloat, defaultValue=6.666, description=""}
  User.MaskParticlesAmount {type=NiagaraFloat, defaultValue=0, description=""}
  User.ParticlesAmount {type=NiagaraFloat, defaultValue=0, description=""}
  User.PowersActive {type=NiagaraBool}
  User.SkeletalMesh {
   type=NiagaraDataInterfaceSkeletalMesh
   defaultValue {
    struct=NiagaraExt_VariableValue_DataInterface
    value {dataInterface=&self:NiagaraDataInterfaceSkeletalMesh_0}
   }
  }
 ]
}
