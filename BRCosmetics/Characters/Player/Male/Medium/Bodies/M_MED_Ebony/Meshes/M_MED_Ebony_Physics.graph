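# Physics asset for the M_MED_Ebony male medium body mesh, providing capsule bodies for the torso, limbs and a two-sided back cape.
# Body capsules cover pelvis, spine, clavicles, arms and legs, with no head, hand or foot bodies.
# Each cape chain hangs from its upper-arm twist body through tiny, free-rotating MSTR and Lookat capsules.
# Four larger capsules follow per side, with swing and twist limits widening down the chain.
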PhysicsAsset /BRCosmetics/Characters/Player/Male/Medium/Bodies/M_MED_Ebony/Meshes/M_MED_Ebony_Physics
bodies {
 calf_l [{shapeType=Capsule, length=26, radius=8, center=<-18 0 0>, rotation=<rot 90 0 0>}]
 calf_r [{shapeType=Capsule, length=26, radius=8, center=<18 0 0>, rotation=<rot 90 0 0>}]
 clavicle_l [{shapeType=Capsule, length=6.9279, radius=7.7908, center=<11.504 0.8386 -1.0247>, rotation=<rot 90 5 0>}]
 clavicle_r [{shapeType=Capsule, length=6.9279, radius=7.7908, center=<-11.504 -0.8386 1.0247>, rotation=<rot 90 5 0>}]
 dyn_cape_bk_l_01 [{shapeType=Capsule, length=12, radius=6, shapeName=dyn_cape_bk_l_01_capsule, center=<6 -5 0>}]
 dyn_cape_bk_l_02 [{shapeType=Capsule, length=11, radius=7, shapeName=dyn_cape_bk_l_02_capsule, center=<6 -5 0>}]
 dyn_cape_bk_l_03 [{shapeType=Capsule, length=10.875, radius=7.125, shapeName=dyn_cape_bk_l_03_capsule, center=<6 -5 0>}]
 dyn_cape_bk_l_04 [{shapeType=Capsule, length=11.375, radius=6.625, shapeName=dyn_cape_bk_l_04_capsule, center=<6 -5 0>}]
 dyn_cape_bk_l_Lookat [{shapeType=Capsule, length=0.505, radius=0.505, shapeName=dyn_cape_bk_l_Lookat_capsule, rotation=<rot 0 -90 0>}]
 dyn_cape_bk_l_MSTR [{shapeType=Capsule, length=0.505, radius=0.505, shapeName=dyn_cape_bk_l_MSTR_capsule, rotation=<rot 0 -90 0>}]
 dyn_cape_bk_r_01 [{shapeType=Capsule, length=12, radius=6, shapeName=dyn_cape_bk_l_01_capsule, center=<6 -5 0>}]
 dyn_cape_bk_r_02 [{shapeType=Capsule, length=11, radius=7, shapeName=dyn_cape_bk_l_02_capsule, center=<6 -5 0>}]
 dyn_cape_bk_r_03 [{shapeType=Capsule, length=10.875, radius=7.125, shapeName=dyn_cape_bk_l_03_capsule, center=<6 -5 0>}]
 dyn_cape_bk_r_04 [{shapeType=Capsule, length=11.375, radius=6.625, shapeName=dyn_cape_bk_l_04_capsule, center=<6 -5 0>}]
 dyn_cape_bk_r_Lookat [{shapeType=Capsule, length=0.505, radius=0.505, shapeName=dyn_cape_bk_r_Lookat_capsule, rotation=<rot 0 -90 0>}]
 dyn_cape_bk_r_MSTR [{shapeType=Capsule, length=0.505, radius=0.505, shapeName=dyn_cape_bk_r_MSTR_capsule, rotation=<rot 0 -90 0>}]
 lowerarm_l [{shapeType=Capsule, length=37.6377, radius=4.63, center=<18.227 0.9178 -2.1609>, rotation=<rot 82.7121 -0.0002 3.2014>}]
 lowerarm_r [{shapeType=Capsule, length=30.5439, radius=5.63, center=<-14.5234 -1.1325 0.8591>, rotation=<rot 82.7121 -0.0002 3.2014>}]
 pelvis [{shapeType=Capsule, length=10.2837, radius=10.6003, center=<-5.0647 -1.3332 0>}]
 spine_01 [{shapeType=Capsule, length=7.63, radius=8.375, shapeName=spine_01_capsule, center=<0 -1.793 0>}]
 spine_02 [{shapeType=Capsule, length=6.5508, radius=8.625, shapeName=spine_02_capsule, center=<-2.2432 -0.4953 0>}]
 spine_03 [{shapeType=Capsule, length=4.0727, radius=9.125}]
 spine_04 [{shapeType=Capsule, length=12.2477, radius=8.375, center=<2.3933 -1.0539 0>}]
 thigh_l [{shapeType=Capsule, length=26, radius=8, center=<-20 0 0>, rotation=<rot 90 0 0>}]
 thigh_r [{shapeType=Capsule, length=26, radius=8, center=<20 0 0>, rotation=<rot 90 0 180>}]
 upperarm_twist_01_l [{shapeType=Capsule, length=24.1779, radius=5.0408, center=<13.2584 -0.8854 -1.0226>, rotation=<rot 87.3404 12.8837 -172.3975>}]
 upperarm_twist_01_r [{shapeType=Capsule, length=24.1779, radius=5.0408, center=<-13.2584 -0.8854 -1.0226>, rotation=<rot 87.3404 12.8837 -172.3975>}]
}
constraints [
 {bone1Name=spine_03, bone2Name=spine_02, swing1LimitDegrees=45, swing2LimitDegrees=45, twistLimitDegrees=45}
 {bone1Name=spine_04, bone2Name=spine_03, swing1LimitDegrees=45, swing2LimitDegrees=45, twistLimitDegrees=45}
 {bone1Name=clavicle_l, bone2Name=spine_04, swing1LimitDegrees=45, swing2LimitDegrees=45, twistLimitDegrees=45}
 {bone1Name=lowerarm_l, bone2Name=clavicle_l, swing1LimitDegrees=45, swing2LimitDegrees=45, twistLimitDegrees=45}
 {bone1Name=upperarm_twist_01_l, bone2Name=clavicle_l, swing1LimitDegrees=45, swing2LimitDegrees=45, twistLimitDegrees=45}
 {bone1Name=clavicle_r, bone2Name=spine_04, swing1LimitDegrees=45, swing2LimitDegrees=45, twistLimitDegrees=45}
 {bone1Name=lowerarm_r, bone2Name=clavicle_r, swing1LimitDegrees=45, swing2LimitDegrees=45, twistLimitDegrees=45}
 {bone1Name=upperarm_twist_01_r, bone2Name=clavicle_r, swing1LimitDegrees=45, swing2LimitDegrees=45, twistLimitDegrees=45}
 {bone1Name=thigh_l, bone2Name=pelvis, swing1LimitDegrees=45, swing2LimitDegrees=45, twistLimitDegrees=45}
 {bone1Name=calf_l, bone2Name=thigh_l, swing1LimitDegrees=45, swing2LimitDegrees=45, twistLimitDegrees=45}
 {bone1Name=thigh_r, bone2Name=pelvis, swing1LimitDegrees=45, swing2LimitDegrees=45, twistLimitDegrees=45}
 {bone1Name=calf_r, bone2Name=thigh_r, swing1LimitDegrees=45, swing2LimitDegrees=45, twistLimitDegrees=45}
 {
  bone1Name=spine_01
  bone2Name=pelvis
  swing1LimitDegrees=45
  swing1Motion=Free
  swing2LimitDegrees=45
  swing2Motion=Free
  twistLimitDegrees=45
  twistMotion=Free
 }
 {
  bone1Name=spine_02
  bone2Name=spine_01
  swing1LimitDegrees=45
  swing1Motion=Free
  swing2LimitDegrees=45
  swing2Motion=Free
  twistLimitDegrees=45
  twistMotion=Free
 }
 {bone1Name=dyn_cape_bk_l_04, bone2Name=dyn_cape_bk_l_03, swing1LimitDegrees=35, swing2LimitDegrees=35, twistLimitDegrees=15}
 {
  bone1Name=dyn_cape_bk_l_MSTR
  bone2Name=upperarm_twist_01_l
  swing1LimitDegrees=45
  swing1Motion=Free
  swing2LimitDegrees=45
  swing2Motion=Free
  twistLimitDegrees=45
  twistMotion=Free
 }
 {
  bone1Name=dyn_cape_bk_l_Lookat
  bone2Name=dyn_cape_bk_l_MSTR
  swing1LimitDegrees=45
  swing1Motion=Free
  swing2LimitDegrees=45
  swing2Motion=Free
  twistLimitDegrees=45
  twistMotion=Free
 }
 {bone1Name=dyn_cape_bk_l_01, bone2Name=dyn_cape_bk_l_Lookat, swing1LimitDegrees=10, swing2LimitDegrees=10, twistLimitDegrees=5}
 {bone1Name=dyn_cape_bk_l_02, bone2Name=dyn_cape_bk_l_01, swing1LimitDegrees=20, swing2LimitDegrees=20, twistLimitDegrees=10}
 {bone1Name=dyn_cape_bk_l_03, bone2Name=dyn_cape_bk_l_02, swing1LimitDegrees=35, swing2LimitDegrees=35, twistLimitDegrees=15}
 {bone1Name=dyn_cape_bk_r_04, bone2Name=dyn_cape_bk_r_03, swing1LimitDegrees=35, swing2LimitDegrees=35, twistLimitDegrees=15}
 {
  bone1Name=dyn_cape_bk_r_MSTR
  bone2Name=upperarm_twist_01_r
  swing1LimitDegrees=45
  swing1Motion=Free
  swing2LimitDegrees=45
  swing2Motion=Free
  twistLimitDegrees=45
  twistMotion=Free
 }
 {
  bone1Name=dyn_cape_bk_r_Lookat
  bone2Name=dyn_cape_bk_r_MSTR
  swing1LimitDegrees=45
  swing1Motion=Free
  swing2LimitDegrees=45
  swing2Motion=Free
  twistLimitDegrees=45
  twistMotion=Free
 }
 {bone1Name=dyn_cape_bk_r_01, bone2Name=dyn_cape_bk_r_Lookat, swing1LimitDegrees=10, swing2LimitDegrees=10, twistLimitDegrees=5}
 {bone1Name=dyn_cape_bk_r_02, bone2Name=dyn_cape_bk_r_01, swing1LimitDegrees=20, swing2LimitDegrees=20, twistLimitDegrees=10}
 {bone1Name=dyn_cape_bk_r_03, bone2Name=dyn_cape_bk_r_02, swing1LimitDegrees=35, swing2LimitDegrees=35, twistLimitDegrees=15}
]
