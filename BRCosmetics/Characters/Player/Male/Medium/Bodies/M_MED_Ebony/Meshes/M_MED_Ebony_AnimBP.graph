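# Animation Blueprint for the M_MED_Ebony cosmetic body, driving cape, strap and flap motion atop the pose copied from its parent mesh.
# Control Rigs bracket weighted constraints, state-offset cape look-at targets, AnimDynamics and two RigidBody simulations.
# Backbling zeroes component-motion scales on the main simulation and half-pins the back cape to the spine.
# Events feed gravity, rifle and cape-variant values, toggling a material element by matching pickaxe and skydiving.
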AnimBlueprint /BRCosmetics/Characters/Player/Male/Medium/Bodies/M_MED_Ebony/Meshes/M_MED_Ebony_AnimBP
target_skeleton=&/M_MED_Ebony_Skeleton
graphs [
 AnimGraph {nodes [
   n1 = "Transform (Modify) Bone - Bone: dyn_cape_fr_r_Lookat"(ComponentPose=Local To Component(), Alpha=1, Translation="-6.424,0.875,0.450", Rotation="0.059,-0.041,-15.003", Scale="1,1,1") {
    rotationMode=BMM_Additive
    rotationSpace=BCS_BoneSpace
    translationMode=BMM_Additive
    translationSpace=BCS_BoneSpace
    alphaBoolBlend {blendOption=Linear}
    boneToModify {boneName=dyn_cape_fr_r_Lookat}
    rotation=<rot 0.059 -0.041 -15.003>
    translation=<-6.424 0.875 0.45>
   }
   Component To Local(ComponentPose=@n1)
   "Constraint - dyn_knee_pad_l (PP)"!(ComponentPose, Alpha=1, ConstraintWeights_0=0.8, ConstraintWeights_1:float=0.2) -> (Pose) {
    alphaBoolBlend {blendOption=Linear}
    boneToModify {boneName=dyn_knee_pad_l}
    constraintSetup [{targetBone {boneName=calf_l}}, {targetBone {boneName=thigh_l}}]
    constraintWeights [1,1]
   }
   n2 = Control Rig(Source=Copy Pose From Mesh() {bUseAttachedParent=true}) {
    alphaBoolBlend {blendOption=Linear}
    controlRigAssetReference {blueprintRigClass=&/M_MED_Ebony_DeformRig_Master_ControlRig.M_MED_Ebony_DeformRig_Master_ControlRig_C}
   }
   n3 = "Constraint - dyn_cape_fr_l_MSTR (PPPPPPPPP)"!(
    ComponentPose=Local To Component(LocalPose=@n2)
    Alpha=1
    ConstraintWeights_0=0.0349
    ConstraintWeights_1:float=0.0496
    ConstraintWeights_2:float=0.1428
    ConstraintWeights_3:float=0.0261
    ConstraintWeights_4:float=0.3284
    ConstraintWeights_5:float=0.1766
    ConstraintWeights_6:float=0.0215
    ConstraintWeights_7:float=0.1728
    ConstraintWeights_8:float=0.0473
   ) -> (Pose) {
    alphaBoolBlend {blendOption=Linear}
    boneToModify {boneName=dyn_cape_fr_l_MSTR}
    constraintSetup [
     {targetBone {boneName=spine_03}}
     {targetBone {boneName=spine_04}}
     {targetBone {boneName=clavicle_l}}
     {targetBone {boneName=upperarm_l}}
     {targetBone {boneName=upperarm_twist_01_l}}
     {targetBone {boneName=deform_upperarm_fr_l}}
     {targetBone {boneName=deform_upperarm_up_l}}
     {targetBone {boneName=upperarm_twist_02_l}}
     {targetBone {boneName=deform_pec_l}}
    ]
    constraintWeights [1,1,1,1,1,1,1,1,1]
   }
   n4 = "Constraint - dyn_cape_fr_r_MSTR (PPPPPPPPP)"!(
    ComponentPose=@n3
    Alpha=1
    ConstraintWeights_0=0.0349
    ConstraintWeights_1:float=0.0496
    ConstraintWeights_2:float=0.1428
    ConstraintWeights_3:float=0.0261
    ConstraintWeights_4:float=0.3284
    ConstraintWeights_5:float=0.1766
    ConstraintWeights_6:float=0.0215
    ConstraintWeights_7:float=0.1728
    ConstraintWeights_8:float=0.0473
   ) -> (Pose) {
    alphaBoolBlend {blendOption=Linear}
    boneToModify {boneName=dyn_cape_fr_r_MSTR}
    constraintSetup [
     {targetBone {boneName=spine_03}}
     {targetBone {boneName=spine_04}}
     {targetBone {boneName=clavicle_r}}
     {targetBone {boneName=upperarm_r}}
     {targetBone {boneName=upperarm_twist_01_r}}
     {targetBone {boneName=deform_upperarm_fr_r}}
     {targetBone {boneName=deform_upperarm_up_r}}
     {targetBone {boneName=upperarm_twist_02_r}}
     {targetBone {boneName=deform_pec_r}}
    ]
    constraintWeights [1,1,1,1,1,1,1,1,1]
   }
   n5 = "Constraint - dyn_cape_bk_l_Lookat (PPPPPPPPPP)"!(
    ComponentPose=@n4
    Alpha=1
    ConstraintWeights_0=0.0054
    ConstraintWeights_1:float=0.1514
    ConstraintWeights_2:float=0.0704
    ConstraintWeights_3:float=0.0534
    ConstraintWeights_4:float=0.2603
    ConstraintWeights_5:float=0.0116
    ConstraintWeights_6:float=0.187
    ConstraintWeights_7:float=0.0187
    ConstraintWeights_8:float=0.1983
    ConstraintWeights_9:float=0.0435
   ) -> (Pose) {
    alphaBoolBlend {blendOption=Linear}
    boneToModify {boneName=dyn_cape_bk_l_Lookat}
    constraintSetup [
     {targetBone {boneName=spine_02}}
     {targetBone {boneName=spine_03}}
     {targetBone {boneName=spine_04}}
     {targetBone {boneName=spine_05}}
     {targetBone {boneName=clavicle_l}}
     {targetBone {boneName=upperarm_l}}
     {targetBone {boneName=upperarm_twist_01_l}}
     {targetBone {boneName=deform_upperarm_up_l}}
     {targetBone {boneName=deform_upperarm_bk_l}}
     {targetBone {boneName=upperarm_twist_02_l}}
    ]
    constraintWeights [1,1,1,1,1,1,1,1,1,1]
   }
   n6 = "Constraint - dyn_cape_bk_r_Lookat (PPPPPPPPPP)"!(
    ComponentPose=@n5
    Alpha=1
    ConstraintWeights_0=0.0054
    ConstraintWeights_1:float=0.1514
    ConstraintWeights_2:float=0.0704
    ConstraintWeights_3:float=0.0534
    ConstraintWeights_4:float=0.2603
    ConstraintWeights_5:float=0.0116
    ConstraintWeights_6:float=0.187
    ConstraintWeights_7:float=0.0187
    ConstraintWeights_8:float=0.1983
    ConstraintWeights_9:float=0.0435
   ) -> (Pose) {
    alphaBoolBlend {blendOption=Linear}
    boneToModify {boneName=dyn_cape_bk_r_Lookat}
    constraintSetup [
     {targetBone {boneName=spine_02}}
     {targetBone {boneName=spine_03}}
     {targetBone {boneName=spine_04}}
     {targetBone {boneName=spine_05}}
     {targetBone {boneName=clavicle_r}}
     {targetBone {boneName=upperarm_r}}
     {targetBone {boneName=upperarm_twist_01_r}}
     {targetBone {boneName=deform_upperarm_up_r}}
     {targetBone {boneName=deform_upperarm_bk_r}}
     {targetBone {boneName=upperarm_twist_02_r}}
    ]
    constraintWeights [1,1,1,1,1,1,1,1,1,1]
   }
   n7 = "Constraint - dyn_strap_fr_l_MSTR (PPPPPPPP)"!(
    ComponentPose=@n6
    Alpha=1
    ConstraintWeights_0=0.0407
    ConstraintWeights_1:float=0.1071
    ConstraintWeights_2:float=0.2445
    ConstraintWeights_3:float=0.2214
    ConstraintWeights_4:float=0.1618
    ConstraintWeights_5:float=0.0115
    ConstraintWeights_6:float=0.0142
    ConstraintWeights_7:float=0.1987
   ) -> (Pose) {
    alphaBoolBlend {blendOption=Linear}
    boneToModify {boneName=dyn_strap_fr_l_MSTR}
    constraintSetup [
     {targetBone {boneName=spine_02}}
     {targetBone {boneName=spine_03}}
     {targetBone {boneName=spine_04}}
     {targetBone {boneName=spine_05}}
     {targetBone {boneName=clavicle_l}}
     {targetBone {boneName=upperarm_twist_01_l}}
     {targetBone {boneName=deform_upperarm_fr_l}}
     {targetBone {boneName=deform_pec_l}}
    ]
    constraintWeights [1,1,1,1,1,1,1,1]
   }
   n8 = "Constraint - dyn_strap_fr_r_MSTR (PPPPPPPP)"!(
    ComponentPose=@n7
    Alpha=1
    ConstraintWeights_0=0.0407
    ConstraintWeights_1:float=0.1071
    ConstraintWeights_2:float=0.2445
    ConstraintWeights_3:float=0.2214
    ConstraintWeights_4:float=0.1618
    ConstraintWeights_5:float=0.0115
    ConstraintWeights_6:float=0.0142
    ConstraintWeights_7:float=0.1987
   ) -> (Pose) {
    alphaBoolBlend {blendOption=Linear}
    boneToModify {boneName=dyn_strap_fr_r_MSTR}
    constraintSetup [
     {targetBone {boneName=spine_02}}
     {targetBone {boneName=spine_03}}
     {targetBone {boneName=spine_04}}
     {targetBone {boneName=spine_05}}
     {targetBone {boneName=clavicle_r}}
     {targetBone {boneName=upperarm_twist_01_r}}
     {targetBone {boneName=deform_upperarm_fr_r}}
     {targetBone {boneName=deform_pec_r}}
    ]
    constraintWeights [1,1,1,1,1,1,1,1]
   }
   n9 = Control Rig!(Source=Component To Local(ComponentPose=@n8), Alpha, cape:double=Get Cape() -> (Cape:double)) -> (Pose) {
    alphaBoolBlend {blendOption=Linear}
    controlRigAssetReference {blueprintRigClass=&/M_MED_Ebony_CtrlRig.M_MED_Ebony_CtrlRig_C}
   }
   n10 = "Transform (Modify) Bone - Bone: cape_TGT"!(
    ComponentPose=Local To Component(LocalPose=@n9)
    bAlphaBoolEnabled:bool=NOT Boolean(A=Get bRigidBodyEnabled())
    Translation="0,-15,0"
    Rotation="0,0,0"
    Scale="1,1,1"
   ) -> (Pose) {
    alphaInputType=Bool
    translationMode=BMM_Additive
    translationSpace=BCS_BoneSpace
    alphaBoolBlend {blendInTime=0.3, blendOption=Linear, blendOutTime=0.3}
    boneToModify {boneName=cape_TGT}
    translation {y=-15}
   }
   n11 = "Transform (Modify) Bone - Bone: cape_TGT"(ComponentPose=@n10, bAlphaBoolEnabled=Get bIsMoving2D(), Translation="0,-100,0", Rotation="0,0,0", Scale="1,1,1") {
    alphaInputType=Bool
    translationMode=BMM_Additive
    translationSpace=BCS_BoneSpace
    alphaBoolBlend {blendInTime=0.3, blendOption=Linear, blendOutTime=0.3}
    boneToModify {boneName=cape_TGT}
    translation {y=-100}
   }
   n12 = "Transform (Modify) Bone - Bone: cape_TGT"(ComponentPose=@n11, bAlphaBoolEnabled=Get bIsSkydiveFloating(), Translation="0,-50,100", Rotation="0,0,0", Scale="1,1,1") {
    alphaInputType=Bool
    translationMode=BMM_Additive
    translationSpace=BCS_BoneSpace
    alphaBoolBlend {blendInTime=0.3, blendOption=Linear, blendOutTime=0.3}
    boneToModify {boneName=cape_TGT}
    translation {y=-50, z=100}
   }
   n13 = "Transform (Modify) Bone - Bone: cape_TGT"(ComponentPose=@n12, bAlphaBoolEnabled=Get bIsSkydiveDiveMode(), Translation="0,125,100", Rotation="0,0,0", Scale="1,1,1") {
    alphaInputType=Bool
    translationMode=BMM_Additive
    translationSpace=BCS_BoneSpace
    alphaBoolBlend {blendInTime=0.3, blendOption=Linear, blendOutTime=0.3}
    boneToModify {boneName=cape_TGT}
    translation {y=125, z=100}
   }
   n14 = "Transform (Modify) Bone - Bone: cape_fr_TGT"!(ComponentPose=@n13, bAlphaBoolEnabled:bool=True, Translation="0,15,0", Rotation="0,0,0", Scale="1,1,1") -> (Pose) {
    alphaInputType=Bool
    translationMode=BMM_Additive
    translationSpace=BCS_BoneSpace
    alphaBoolBlend {blendInTime=0.3, blendOption=Linear, blendOutTime=0.3}
    boneToModify {boneName=cape_fr_TGT}
    translation {y=15}
   }
   n15 = "Transform (Modify) Bone - Bone: cape_fr_TGT"(ComponentPose=@n14, bAlphaBoolEnabled=Get bIsSkydiveFloating(), Translation="0,-40,35", Rotation="0,0,0", Scale="1,1,1") {
    alphaInputType=Bool
    translationMode=BMM_Additive
    translationSpace=BCS_BoneSpace
    alphaBoolBlend {blendInTime=0.3, blendOption=Linear, blendOutTime=0.3}
    boneToModify {boneName=cape_fr_TGT}
    translation {y=-40, z=35}
   }
   n16 = "Transform (Modify) Bone - Bone: cape_fr_TGT"(ComponentPose=@n15, bAlphaBoolEnabled=Get bIsSkydiveDiveMode(), Translation="0,-60,125", Rotation="0,0,0", Scale="1,1,1") {
    alphaInputType=Bool
    translationMode=BMM_Additive
    translationSpace=BCS_BoneSpace
    alphaBoolBlend {blendInTime=0.3, blendOption=Linear, blendOutTime=0.3}
    boneToModify {boneName=cape_fr_TGT}
    translation {y=-60, z=125}
   }
   n17 = Look At - Bone: dyn_cape_bk_l_Lookat(ComponentPose=@n16, Alpha=1) {
    alphaBoolBlend {blendOption=Linear}
    boneToModify {boneName=dyn_cape_bk_l_Lookat}
    lookAtLocation {x=0}
    lookAtTarget {boneReference {boneName=cape_TGT}}
    lookAt_Axis {axis=<1 0>}
   }
   n18 = Look At - Bone: dyn_cape_bk_r_Lookat(ComponentPose=@n17, Alpha=1) {
    alphaBoolBlend {blendOption=Linear}
    boneToModify {boneName=dyn_cape_bk_r_Lookat}
    lookAtLocation {x=0}
    lookAtTarget {boneReference {boneName=cape_TGT}}
    lookAt_Axis {axis=<1 0>}
   }
   n19 = NOT Boolean(A=Get bRigidBodyEnabled())
   n20 = Look At - Bone: dyn_cape_fr_l_Lookat!(ComponentPose=@n18, bAlphaBoolEnabled:bool=@n19) -> (Pose) {
    alphaInputType=Bool
    alphaBoolBlend {blendInTime=0.1, blendOption=Linear, blendOutTime=0.1}
    boneToModify {boneName=dyn_cape_fr_l_Lookat}
    lookAtLocation {x=0}
    lookAtTarget {boneReference {boneName=cape_fr_TGT}}
    lookAt_Axis {axis=<1 0>}
   }
   n21 = Look At - Bone: dyn_cape_fr_r_Lookat!(ComponentPose=@n20, bAlphaBoolEnabled:bool=@n19) -> (Pose) {
    alphaInputType=Bool
    alphaBoolBlend {blendInTime=0.1, blendOption=Linear, blendOutTime=0.1}
    boneToModify {boneName=dyn_cape_fr_r_Lookat}
    lookAtLocation {x=0}
    lookAtTarget {boneReference {boneName=cape_fr_TGT}}
    lookAt_Axis {axis=<1 0>}
   }
   n22 = Anim Dynamics - Bone: dyn_flap_bk_l_01(ComponentPose=@n21, Alpha=0.5, ExternalForce="0,0,0") {
    angularDampingOverride=0.98
    angularSpringConstant=50
    bAngularSpring=true
    bGravityOverrideInSimSpace=true
    bOverrideAngularDamping=true
    simulationSpace=RootRelative
    alphaBoolBlend {blendOption=Linear}
    boundBone {boneName=dyn_flap_bk_l_01}
    componentAppliedLinearAccClamp=<3500 3500 3500>
    componentLinearAccScale=<0.6 0.6 1.2>
    componentLinearVelScale=<0.6 0.6 1.2>
    physicsBodyDefinitions [{
      sphereCollisionRadius=0
      boundBone {boneName=dyn_flap_bk_l_01}
      boxExtents=<17 17 17>
      constraintSetup {angularLimitsMax=<1 15 5>, angularLimitsMin=<-1 -15 -10>, angularTarget=<1 0 0>}
      localJointOffset=<1 0 0>
     }]
    relativeSpaceBone {boneName=root}
   }
   n23 = Anim Dynamics - Bone: dyn_flap_bk_l_02(ComponentPose=@n22, Alpha=0.5, ExternalForce="0,0,0") {
    angularDampingOverride=0.98
    angularSpringConstant=50
    bAngularSpring=true
    bGravityOverrideInSimSpace=true
    bOverrideAngularDamping=true
    simulationSpace=RootRelative
    alphaBoolBlend {blendOption=Linear}
    boundBone {boneName=dyn_flap_bk_l_02}
    componentAppliedLinearAccClamp=<3500 3500 3500>
    componentLinearAccScale=<0.6 0.6 1.2>
    componentLinearVelScale=<0.6 0.6 1.2>
    physicsBodyDefinitions [{
      sphereCollisionRadius=0
      boundBone {boneName=dyn_flap_bk_l_02}
      boxExtents=<17 17 17>
      constraintSetup {angularLimitsMax=<1 15 5>, angularLimitsMin=<-1 -15 -10>, angularTarget=<1 0 0>}
      localJointOffset=<1 0 0>
     }]
    relativeSpaceBone {boneName=root}
   }
   n24 = Anim Dynamics - Bone: dyn_flap_bk_r_01(ComponentPose=@n23, Alpha=0.5, ExternalForce="0,0,0") {
    angularDampingOverride=0.98
    angularSpringConstant=50
    bAngularSpring=true
    bGravityOverrideInSimSpace=true
    bOverrideAngularDamping=true
    simulationSpace=RootRelative
    alphaBoolBlend {blendOption=Linear}
    boundBone {boneName=dyn_flap_bk_r_01}
    componentAppliedLinearAccClamp=<3500 3500 3500>
    componentLinearAccScale=<0.6 0.6 1.2>
    componentLinearVelScale=<0.6 0.6 1.2>
    physicsBodyDefinitions [{
      sphereCollisionRadius=0
      boundBone {boneName=dyn_flap_bk_r_01}
      boxExtents=<17 17 17>
      constraintSetup {angularLimitsMax=<1 15 5>, angularLimitsMin=<-1 -15 -10>, angularTarget=<1 0 0>}
      localJointOffset=<1 0 0>
     }]
    relativeSpaceBone {boneName=root}
   }
   n25 = Anim Dynamics - Bone: dyn_flap_bk_r_02(ComponentPose=@n24, Alpha=0.5, ExternalForce="0,0,0") {
    angularDampingOverride=0.98
    angularSpringConstant=50
    bAngularSpring=true
    bGravityOverrideInSimSpace=true
    bOverrideAngularDamping=true
    simulationSpace=RootRelative
    alphaBoolBlend {blendOption=Linear}
    boundBone {boneName=dyn_flap_bk_r_02}
    componentAppliedLinearAccClamp=<3500 3500 3500>
    componentLinearAccScale=<0.6 0.6 1.2>
    componentLinearVelScale=<0.6 0.6 1.2>
    physicsBodyDefinitions [{
      sphereCollisionRadius=0
      boundBone {boneName=dyn_flap_bk_r_02}
      boxExtents=<17 17 17>
      constraintSetup {angularLimitsMax=<1 15 5>, angularLimitsMin=<-1 -15 -10>, angularTarget=<1 0 0>}
      localJointOffset=<1 0 0>
     }]
    relativeSpaceBone {boneName=root}
   }
   n26 = Get bIsInFrontEnd()
   n27 = "Transform (Modify) Bone - Bone: dyn_cape_bk_l_01"!(ComponentPose=@n25, bAlphaBoolEnabled:bool=@n26, Translation="0,0,0", Rotation="0,-25,0", Scale="1,1,1") -> (Pose) {
    alphaInputType=Bool
    rotationMode=BMM_Additive
    rotationSpace=BCS_BoneSpace
    translationMode=BMM_Additive
    translationSpace=BCS_BoneSpace
    alphaBoolBlend {blendOption=Linear}
    boneToModify {boneName=dyn_cape_bk_l_01}
    rotation {yaw=-25}
   }
   n28 = "Transform (Modify) Bone - Bone: dyn_cape_bk_r_01"!(ComponentPose=@n27, bAlphaBoolEnabled:bool=@n26, Translation="0,0,0", Rotation="0,-25,0", Scale="1,1,1") -> (Pose) {
    alphaInputType=Bool
    rotationMode=BMM_Additive
    rotationSpace=BCS_BoneSpace
    translationMode=BMM_Additive
    translationSpace=BCS_BoneSpace
    alphaBoolBlend {blendOption=Linear}
    boneToModify {boneName=dyn_cape_bk_r_01}
    rotation {yaw=-25}
   }
   n29 = Get GravityOverride()
   n30 = RigidBody!(
    ComponentPose=@n28
    Alpha
    OverrideWorldGravity:Vector=@n29
    ExternalForce
    ComponentLinearAccScale:Vector=Get ACC() -> (ACC:Vector)
    ComponentLinearVelScale:Vector=Get VEL() -> (VEL:Vector)
   ) -> (Pose) {
    bClampLinearTranslationLimitToRefPose=true
    bOverrideWorldGravity=true
    overridePhysicsAsset=&/M_MED_Ebony_Physics
    simulationSpace=BaseBoneSpace
    alphaBoolBlend {blendOption=Linear}
    baseBoneRef {boneName=root}
    componentAppliedLinearAccClamp=<1500 1500 1500>
    componentLinearAccScale=<0.3 0.3 0.5>
    componentLinearVelScale=<0.3 0.3 0.3>
   }
   n31 = RigidBody!(ComponentPose=@n30, Alpha, OverrideWorldGravity=@n29, ExternalForce) -> (Pose) {
    bClampLinearTranslationLimitToRefPose=true
    bOverrideWorldGravity=true
    overridePhysicsAsset=&/M_MED_Ebony_Front_Physics
    simulationSpace=BaseBoneSpace
    alphaBoolBlend {blendOption=Linear}
    baseBoneRef {boneName=pelvis}
    componentAppliedLinearAccClamp=<3000 3000 3000>
    componentLinearAccScale=<0.3 0.3 0.3>
    componentLinearVelScale=<0.3 0.3 0.3>
   }
   n32 = Get backpack() -> (backpack:double)
   n33 = "Constraint - dyn_cape_bk_l_02 (P)"(ComponentPose=@n31, Alpha=@n32, ConstraintWeights_0=1) {
    alphaBoolBlend {blendOption=Linear}
    alphaScaleBiasClamp {bInterpResult=true}
    boneToModify {boneName=dyn_cape_bk_l_02}
    constraintSetup [{targetBone {boneName=spine_04}}]
    constraintWeights [1]
   }
   n34 = "Constraint - dyn_cape_bk_l_03 (P)"(ComponentPose=@n33, Alpha=@n32, ConstraintWeights_0=1) {
    alphaBoolBlend {blendOption=Linear}
    alphaScaleBiasClamp {bInterpResult=true}
    boneToModify {boneName=dyn_cape_bk_l_03}
    constraintSetup [{targetBone {boneName=spine_02}}]
    constraintWeights [1]
   }
   n35 = "Constraint - dyn_cape_bk_r_02 (P)"(ComponentPose=@n34, Alpha=@n32, ConstraintWeights_0=1) {
    alphaBoolBlend {blendOption=Linear}
    alphaScaleBiasClamp {bInterpResult=true}
    boneToModify {boneName=dyn_cape_bk_r_02}
    constraintSetup [{targetBone {boneName=spine_04}}]
    constraintWeights [1]
   }
   n36 = "Constraint - dyn_cape_bk_r_03 (P)"(ComponentPose=@n35, Alpha=@n32, ConstraintWeights_0=1) {
    alphaBoolBlend {blendOption=Linear}
    alphaScaleBiasClamp {bInterpResult=true}
    boneToModify {boneName=dyn_cape_bk_r_03}
    constraintSetup [{targetBone {boneName=spine_02}}]
    constraintWeights [1]
   }
   n37 = NOT Boolean(A=Get bRigidBodyEnabled())
   n38 = "Constraint - dyn_strap_fr_l_02 (P)"!(ComponentPose=@n36, bAlphaBoolEnabled:bool=@n37, ConstraintWeights_0=1) -> (Pose) {
    alphaInputType=Bool
    alphaBoolBlend {blendInTime=0.1, blendOption=Linear, blendOutTime=0.1}
    boneToModify {boneName=dyn_strap_fr_l_02}
    constraintSetup [{targetBone {boneName=spine_04}}]
    constraintWeights [1]
   }
   n39 = "Constraint - dyn_strap_fr_l_03 (P)"!(ComponentPose=@n38, bAlphaBoolEnabled:bool=@n37, ConstraintWeights_0=1) -> (Pose) {
    alphaInputType=Bool
    alphaBoolBlend {blendInTime=0.1, blendOption=Linear, blendOutTime=0.1}
    boneToModify {boneName=dyn_strap_fr_l_03}
    constraintSetup [{targetBone {boneName=spine_03}}]
    constraintWeights [1]
   }
   n40 = "Constraint - dyn_strap_fr_l_04 (P)"!(ComponentPose=@n39, bAlphaBoolEnabled:bool=@n37, ConstraintWeights_0=1) -> (Pose) {
    alphaInputType=Bool
    alphaBoolBlend {blendInTime=0.1, blendOption=Linear, blendOutTime=0.1}
    boneToModify {boneName=dyn_strap_fr_l_04}
    constraintSetup [{targetBone {boneName=spine_02}}]
    constraintWeights [1]
   }
   n41 = "Constraint - dyn_strap_fr_r_02 (P)"!(ComponentPose=@n40, bAlphaBoolEnabled:bool=@n37, ConstraintWeights_0=1) -> (Pose) {
    alphaInputType=Bool
    alphaBoolBlend {blendInTime=0.1, blendOption=Linear, blendOutTime=0.1}
    boneToModify {boneName=dyn_strap_fr_r_02}
    constraintSetup [{targetBone {boneName=spine_04}}]
    constraintWeights [1]
   }
   n42 = "Constraint - dyn_strap_fr_r_03 (P)"!(ComponentPose=@n41, bAlphaBoolEnabled:bool=@n37, ConstraintWeights_0=1) -> (Pose) {
    alphaInputType=Bool
    alphaBoolBlend {blendInTime=0.1, blendOption=Linear, blendOutTime=0.1}
    boneToModify {boneName=dyn_strap_fr_r_03}
    constraintSetup [{targetBone {boneName=spine_03}}]
    constraintWeights [1]
   }
   n43 = "Constraint - dyn_strap_fr_r_04 (P)"!(ComponentPose=@n42, bAlphaBoolEnabled:bool=@n37, ConstraintWeights_0=1) -> (Pose) {
    alphaInputType=Bool
    alphaBoolBlend {blendInTime=0.1, blendOption=Linear, blendOutTime=0.1}
    boneToModify {boneName=dyn_strap_fr_r_04}
    constraintSetup [{targetBone {boneName=spine_02}}]
    constraintWeights [1]
   }
   n44 = Control Rig!(Source=Component To Local(ComponentPose=@n43), Alpha, rifle:double=Get Rifle_alpha() -> (Rifle_alpha:double), Cape:double=Get Cape()) -> (Pose) {
    alphaBoolBlend {blendOption=Linear}
    controlRigAssetReference {blueprintRigClass=&/M_MED_Ebony_post_CtrlRig.M_MED_Ebony_post_CtrlRig_C}
   }
   Output Pose(Result=@n44) {layerGroup=None}
  ]}
 EventGraph {nodes [
   Event BlueprintUpdateAnimation()
   > n1 = Sequence!(execute) -> (then_0, then_1, then_2:Exec)
   > n2 = Gravity Override State Function DEPRECATED(CharacterPartAnimInstance=Self-Reference(), GameTimeInSeconds=GetGameTimeInSeconds(), GravityOverrideParams=Get Gravity Override Params())
   > Set GravityOverride(GravityOverride=@n2)
   > n3 = Branch(Condition=CheckForBackbling())
   > Set ACC(ACC:Vector="0, 0, 0") -> (Output_Get:Vector)
   > Set VEL(VEL:Vector="0, 0, 0") -> (Output_Get:Vector)
   > Set backpack(backpack:double=0.5) -> (Output_Get:double)
   Set ACC(execute=@n3.else, ACC="0.3,0.3,0.5")
   > Set VEL(VEL="0.3,0.3,0.3")
   > Set backpack(backpack=0)
   n4 = Cast To FortPlayerAnimInstance(execute=@n1.then_1, Object=GetAnimInstance(self=GetOwnerSkeletalMeshComponent()))
   > Set Rifle(Rifle:bool=false) -> (Output_Get:bool)
   > n5 = Switch on EFortWeaponCoreAnimation!(execute, Selection:EFortWeaponCoreAnimation Enum=Get WeaponCoreAnim(self=@n4)) -> (
    Melee:Exec
    Pistol:Exec
    Shotgun:Exec
    PaperBlueprint:Exec
    Rifle:Exec
    MeleeOneHand:Exec
    MachinePistol:Exec
    RocketLauncher:Exec
    GrenadeLauncher:Exec
    GoingCommando:Exec
    AssaultRifle:Exec
    TacticalShotgun:Exec
    SniperRifle:Exec
    TrapPlacement:Exec
    ShoulderLauncher:Exec
    AbilityDecoTool:Exec
    Crossbow:Exec
    C4:Exec
    RemoteControl:Exec
    DualWield:Exec
    AR_BullPup:Exec
    AR_ForwardGrip:Exec
    MedPackPaddles:Exec
    SMG_P90:Exec
    AR_DrumGun:Exec
    Consumable_Small:Exec
    Consumable_Large:Exec
    Balloon:Exec
    MountedTurret:Exec
    CreativeTool:Exec
    ExplosiveBow:Exec
    AshtonIndigo:Exec
    AshtonChicago:Exec
    MeleeDualWield:Exec
    Unarmed:Exec
   )
   Branch(execute=[
     @n5.AR_BullPup
     @n5.AR_ForwardGrip
     @n5.AssaultRifle
     @n5.Crossbow
     @n5.GrenadeLauncher
     @n5.Rifle
     @n5.RocketLauncher
     @n5.Shotgun
     @n5.SniperRifle
     @n5.TacticalShotgun
    ])
   > Set Rifle(Rifle=true)
   n6 = Branch(execute=@n1.then_2, Condition=Get Rifle() -> (Rifle:bool))
   > Set Rifle_alpha(Rifle_alpha:double=1) -> (Output_Get:double)
   Set Rifle_alpha(execute=@n6.else, Rifle_alpha=0)
   Event OnCharacterCustomizationLoaded()
   > n7 = Sequence()
   n8 = Get Dynamic Material Instance(execute=[@n15.then, @n7.then_0], SkeletalMeshComponent=GetPartSkeletalMeshComponent())
   n9 = PickaxeItemDefinitionHasAnyMetaTag(Tags=Make Array("[0]"=(TagName="Cosmetics.MetaTags.Ebony")), bRequiresEquipped=true)
   > n10 = Branch(Condition=@n9)
   > n11 = Branch(Condition=Get bIsSkyDiving())
   SetScalarParameterValue(execute=[@n13.then, @n14.else, @n10.else, @n11.then], self=@n8, ParameterName=Hide Element 02, Value=0)
   SetScalarParameterValue(execute=[@n14.then, @n11.else], self=@n8, ParameterName=Hide Element 02, Value=1)
   n12 = Branch(execute=@n7.then_1, Condition=CharacterItemDefinitionHasAnyMetaTag(Tags=Make Array("[0]"=(TagName="Cosmetics.Variant.Property.Stage3"))))
   > Set Cape(Cape:double=1) -> (Output_Get:double)
   Set Cape(execute=@n12.else, Cape=0)
   Delay(execute=@n7.then_2, Duration=0.1)
   > ResetDynamics(InTeleportType=TeleportPhysics)
   > ResetDynamics(InTeleportType=ResetPhysics)
   > Delay(Duration=0.1)
   n13 = Event OnBeginSkydiving()
   Event OnEndSkydiving()
   > n14 = Branch(Condition=@n9)
   n15 = Event OnWeaponChanged()
  ]}
]
variables [
 ACC {type=Vector}
 backpack {type=number}
 Cape {type=number}
 frontendLoad {type=boolean}
 Gravity Override Params {type=GravityOverrideParamsStruct}
 GravityOverride {type=Vector}
 Rifle {type=boolean}
 Rifle_alpha {type=number}
 VEL {type=Vector}
]
