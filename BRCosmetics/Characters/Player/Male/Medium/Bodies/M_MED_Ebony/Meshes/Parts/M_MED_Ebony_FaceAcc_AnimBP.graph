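# Post-process style secondary-motion AnimBlueprint for the Ebony male character's FaceAcc part, driving hood and front strap bones.
# Copied parent and body_mesh poses pass through a deform Control Rig and weighted strap constraints into a cached pose.
# bRigidBodyEnabled selects RigidBody simulation or an AnimDynamics hood chain, followed by the accessory's Control Rig.
# EventGraph lowers RigidBody alpha while aiming and flags rifle-like weapons for that final rig.
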AnimBlueprint /BRCosmetics/Characters/Player/Male/Medium/Bodies/M_MED_Ebony/Meshes/Parts/M_MED_Ebony_FaceAcc_AnimBP
target_skeleton=&/M_MED_Ebony_FaceAcc_Skeleton
graphs [
 AnimGraph {nodes [
   Anim Dynamics - Bone: dyn_hood_bk!(ComponentPose, GravityOverride:Vector="0,0,-980", ExternalForce="0,0,0") -> (Pose) {
    alpha=0.2
    angularDampingOverride=0.99
    angularSpringConstant=5
    bAngularSpring=true
    bGravityOverrideInSimSpace=true
    bLinearSpring=true
    bOverrideAngularDamping=true
    bUseGravityOverride=true
    gravityScale=0.35
    linearDampingOverride=0.99
    linearSpringConstant=10
    simulationSpace=RootRelative
    alphaBoolBlend {blendOption=Linear}
    alphaScaleBiasClamp {clampMax=0.5}
    boundBone {boneName=dyn_hood_bk}
    componentAppliedLinearAccClamp=<2500 2500 2500>
    componentLinearAccScale=<0.6 0.5 0.5>
    componentLinearVelScale=<0.6 0.5 0.5>
    gravityOverride {z=-980}
    physicsBodyDefinitions [{
      sphereCollisionRadius=0
      boundBone {boneName=dyn_hood_bk}
      boxExtents=<10 5 5>
      constraintSetup {angularLimitsMax=<1 -5 2>, angularLimitsMin=<-1 -5 -2>, angularTarget=<1 0 0>, linearAxesMax=<3 3 4>, linearAxesMin=<-2 -2 -5>}
      localJointOffset=<5 0 0>
     }]
   }
   RigidBody!(ComponentPose, Alpha, OverrideWorldGravity:Vector="0,0,-980", ExternalForce) -> (Pose) {
    bClampLinearTranslationLimitToRefPose=true
    bOverrideWorldGravity=true
    overridePhysicsAsset=&/M_MED_Ebony_FaceAcc_Physics
    simulationSpace=BaseBoneSpace
    alphaBoolBlend {blendOption=Linear}
    alphaScaleBiasClamp {interpSpeedDecreasing=4, interpSpeedIncreasing=4}
    baseBoneRef {boneName=spine_05}
    componentAppliedLinearAccClamp=<6000 6000 6000>
    componentLinearAccScale=<0.25 0.25 0.25>
    componentLinearVelScale=<0.25 0.25 0.25>
    overrideWorldGravity {z=-980}
   }
   n1 = RigidBody(ComponentPose=Local To Component(LocalPose=Use cached pose 'Pose'()), Alpha=Get ADS() -> (ADS:double), OverrideWorldGravity="0,0,-980") {
    bClampLinearTranslationLimitToRefPose=true
    bOverrideWorldGravity=true
    overridePhysicsAsset=&M_MED_Ebony_FaceAcc_Physics
    simulationSpace=BaseBoneSpace
    alphaBoolBlend {blendOption=Linear}
    alphaScaleBiasClamp {interpSpeedDecreasing=4, interpSpeedIncreasing=4}
    baseBoneRef {boneName=spine_05}
    componentAppliedLinearAccClamp=<6000 6000 6000>
    componentLinearAccScale=<0.2 0.2 0.1>
    componentLinearVelScale=<0.2 0.2 0.1>
    overrideWorldGravity {z=-980}
   }
   n2 = Anim Dynamics - Bone: dyn_hood_l!(ComponentPose=Local To Component(LocalPose=Use cached pose 'Pose'()), Alpha=1, GravityOverride:Vector="0,0,-980", ExternalForce="0,0,0") -> (Pose) {
    angularDampingOverride=1
    angularSpringConstant=20000
    bAngularSpring=true
    bOverrideAngularDamping=true
    bUseGravityOverride=true
    gravityScale=0.35
    simulationSpace=RootRelative
    alphaBoolBlend {blendOption=Linear}
    boundBone {boneName=dyn_hood_l}
    componentAppliedLinearAccClamp=<3000 3000 3000>
    componentLinearAccScale=<0.25 0.25 0.25>
    componentLinearVelScale=<0.25 0.25 0.25>
    gravityOverride {z=-980}
    physicsBodyDefinitions [{
      boundBone {boneName=dyn_hood_l}
      boxExtents=<10 10 10>
      constraintSetup {angularLimitsMax=<10 15 20>, angularLimitsMin=<-10 -25 -20>, angularTarget=<0 0 0>, linearAxesMax=<3 2 8>, linearAxesMin=<-2 -3 -5>}
      localJointOffset=<0 0 0>
     }]
   }
   n3 = Anim Dynamics - Bone: dyn_hood_r!(ComponentPose=@n2, Alpha=1, GravityOverride:Vector="0,0,-980", ExternalForce="0,0,0") -> (Pose) {
    angularDampingOverride=1
    angularSpringConstant=20000
    bAngularSpring=true
    bOverrideAngularDamping=true
    bUseGravityOverride=true
    gravityScale=0.35
    simulationSpace=RootRelative
    alphaBoolBlend {blendOption=Linear}
    boundBone {boneName=dyn_hood_r}
    componentAppliedLinearAccClamp=<3000 3000 3000>
    componentLinearAccScale=<0.25 0.25 0.25>
    componentLinearVelScale=<0.25 0.25 0.25>
    gravityOverride {z=-980}
    physicsBodyDefinitions [{
      boundBone {boneName=dyn_hood_r}
      boxExtents=<10 10 10>
      constraintSetup {angularLimitsMax=<10 15 20>, angularLimitsMin=<-10 -25 -20>, angularTarget=<0 0 0>, linearAxesMax=<3 3 8>, linearAxesMin=<-2 -2 -5>}
      localJointOffset=<0 0 0>
     }]
   }
   n4 = Anim Dynamics - Bone: dyn_hood_fr!(ComponentPose=@n3, GravityOverride:Vector="0,0,-980", ExternalForce="0,0,0") -> (Pose) {
    alpha=0.6
    angularDampingOverride=0.99
    angularSpringConstant=50
    bAngularSpring=true
    bGravityOverrideInSimSpace=true
    bOverrideAngularDamping=true
    bUseGravityOverride=true
    gravityScale=0.35
    linearDampingOverride=0.99
    linearSpringConstant=10
    simulationSpace=RootRelative
    alphaBoolBlend {blendOption=Linear}
    boundBone {boneName=dyn_hood_fr}
    componentAppliedLinearAccClamp=<3000 3000 3000>
    componentLinearAccScale=<0.6 0.6 1.2>
    componentLinearVelScale=<0.6 0.6 1.2>
    gravityOverride {z=-980}
    physicsBodyDefinitions [{
      sphereCollisionRadius=0
      boundBone {boneName=dyn_hood_fr}
      boxExtents=<20 15 15>
      constraintSetup {angularLimitsMax=<1 15 2>, angularLimitsMin=<-1 -15 -5>, angularTarget=<1 0 0>, linearAxesMax=<0 0 3>, linearAxesMin=<-1 0 0>}
      localJointOffset=<5 0 0>
     }]
   }
   n5 = Anim Dynamics - Bone: dyn_hood_bk(ComponentPose=@n4, GravityOverride="0,0,-980", ExternalForce="0,0,0") {
    alpha=0.6
    angularDampingOverride=0.99
    angularSpringConstant=50
    bAngularSpring=true
    bGravityOverrideInSimSpace=true
    bOverrideAngularDamping=true
    bUseGravityOverride=true
    gravityScale=0.35
    linearDampingOverride=0.99
    linearSpringConstant=10
    simulationSpace=RootRelative
    alphaBoolBlend {blendOption=Linear}
    boundBone {boneName=dyn_hood_bk}
    componentAppliedLinearAccClamp=<3500 3500 3500>
    componentLinearAccScale=<0.8 1.2 1.2>
    componentLinearVelScale=<0.8 1.2 1.2>
    gravityOverride {z=-980}
    physicsBodyDefinitions [{
      sphereCollisionRadius=0
      boundBone {boneName=dyn_hood_bk}
      boxExtents=<20 15 15>
      constraintSetup {angularLimitsMax=<1 15 5>, angularLimitsMin=<-1 -15 -10>, angularTarget=<1 0 0>}
      localJointOffset=<5 0 0>
     }]
   }
   n6 = Control Rig!(
    Source=Blend Poses by bool(BlendPose_0=Component To Local(ComponentPose=@n1), BlendPose_1=Component To Local(ComponentPose=@n5), bActiveValue=Get bRigidBodyEnabled()) {blendType=Linear}
    Alpha
    Rifle:double=Get Rifle_alpha() -> (Rifle_alpha:double)
   ) -> (Pose) {
    alphaBoolBlend {blendOption=Linear}
    controlRigAssetReference {blueprintRigClass=&/M_MED_Ebony_FaceAcc_CtrlRig.M_MED_Ebony_FaceAcc_CtrlRig_C}
   }
   Output Pose(Result=@n6) {layerGroup=None}
   n7 = Layered blend per bone(
    BasePose=Copy Pose From Mesh() {bCopyCurves=true, bUseAttachedParent=true}
    BlendPoses_0=Copy Pose From Mesh(SourceMeshComponent=Get body_mesh() -> (body_mesh:Skeletal Mesh Component Object Reference))
   ) {layerSetup [{branchFilters [{boneName=upperarm_r}, {boneName=upperarm_l}, {boneName=dyn_strap_fr_l_MSTR}, {boneName=dyn_strap_fr_r_MSTR}]}]}
   n8 = Control Rig!(Source=@n7, Alpha) -> (Pose) {
    alphaBoolBlend {blendOption=Linear}
    controlRigAssetReference {blueprintRigClass=/BRCosmetics/Characters/Player/Male/Medium/Bodies/M_MED_Ebony/Meshes/M_MED_Ebony_DeformRig_Master_ControlRig.M_MED_Ebony_DeformRig_Master_ControlRig_C}
   }
   n9 = "Constraint - dyn_strap_fr_l_MSTR (PPPPPPPP)"!(
    ComponentPose=Local To Component(LocalPose=@n8)
    Alpha=1
    ConstraintWeights_0=0.0407
    ConstraintWeights_1:float=0.1071
    ConstraintWeights_2:float=0.2445
    ConstraintWeights_3:float=0.2214
    ConstraintWeights_4:float=0.1618
    ConstraintWeights_5:float=0.0115
    ConstraintWeights_6:float=0.0142
    ConstraintWeights_7:float=0.1987
   ) -> (Pose) {
    alphaBoolBlend {blendOption=Linear}
    boneToModify {boneName=dyn_strap_fr_l_MSTR}
    constraintSetup [
     {targetBone {boneName=spine_02}}
     {targetBone {boneName=spine_03}}
     {targetBone {boneName=spine_04}}
     {targetBone {boneName=spine_05}}
     {targetBone {boneName=clavicle_l}}
     {targetBone {boneName=upperarm_twist_01_l}}
     {targetBone {boneName=deform_upperarm_fr_l}}
     {targetBone {boneName=deform_pec_l}}
    ]
    constraintWeights [1,1,1,1,1,1,1,1]
   }
   n10 = "Constraint - dyn_strap_fr_r_MSTR (PPPPPPPP)"!(
    ComponentPose=@n9
    Alpha=1
    ConstraintWeights_0=0.0407
    ConstraintWeights_1:float=0.1071
    ConstraintWeights_2:float=0.2445
    ConstraintWeights_3:float=0.2214
    ConstraintWeights_4:float=0.1618
    ConstraintWeights_5:float=0.0115
    ConstraintWeights_6:float=0.0142
    ConstraintWeights_7:float=0.1987
   ) -> (Pose) {
    alphaBoolBlend {blendOption=Linear}
    boneToModify {boneName=dyn_strap_fr_r_MSTR}
    constraintSetup [
     {targetBone {boneName=spine_02}}
     {targetBone {boneName=spine_03}}
     {targetBone {boneName=spine_04}}
     {targetBone {boneName=spine_05}}
     {targetBone {boneName=clavicle_r}}
     {targetBone {boneName=upperarm_twist_01_r}}
     {targetBone {boneName=deform_upperarm_fr_r}}
     {targetBone {boneName=deform_pec_r}}
    ]
    constraintWeights [1,1,1,1,1,1,1,1]
   }
   Save cached pose 'Pose'(Pose=Component To Local(ComponentPose=@n10)) {cacheName=Pose}
  ]}
 EventGraph {nodes [
   Event BlueprintInitializeAnimation()
   > Set Body_Mesh(Body_Mesh=GetOwnerSkeletalMeshComponent())
   Event BlueprintUpdateAnimation()
   > n1 = Sequence!(execute) -> (then_0, then_1, then_2:Exec, then_3:Exec, then_4:Exec)
   > n2 = Gravity Override State Function DEPRECATED(CharacterPartAnimInstance=Self-Reference(), GameTimeInSeconds=GetGameTimeInSeconds(), GravityOverrideParams=Get Gravity Override Params())
   > Set GravityOverride(GravityOverride=@n2)
   n3 = Cast To PlayerPawn_Athena(execute=@n1.then_1, Object=TryGetPawnOwner())
   > Set face_mesh(face_mesh:Skeletal Mesh Component Object Reference=GetSkeletalMeshForPartType(self=@n3, PartType=Face)) -> (Output_Get:Skeletal Mesh Component Object Reference)
   n4 = Cast To FortPlayerAnimInstance(execute=@n1.then_2, Object=GetAnimInstance(self=GetOwnerSkeletalMeshComponent()))
   > Set Rifle(Rifle:bool=false) -> (Output_Get:bool)
   > n5 = Switch on EFortWeaponCoreAnimation!(execute, Selection:EFortWeaponCoreAnimation Enum=Get WeaponCoreAnim(self=@n4)) -> (
    Melee:Exec
    Pistol:Exec
    Shotgun:Exec
    PaperBlueprint:Exec
    Rifle:Exec
    MeleeOneHand:Exec
    MachinePistol:Exec
    RocketLauncher:Exec
    GrenadeLauncher:Exec
    GoingCommando:Exec
    AssaultRifle:Exec
    TacticalShotgun:Exec
    SniperRifle:Exec
    TrapPlacement:Exec
    ShoulderLauncher:Exec
    AbilityDecoTool:Exec
    Crossbow:Exec
    C4:Exec
    RemoteControl:Exec
    DualWield:Exec
    AR_BullPup:Exec
    AR_ForwardGrip:Exec
    MedPackPaddles:Exec
    SMG_P90:Exec
    AR_DrumGun:Exec
    Consumable_Small:Exec
    Consumable_Large:Exec
    Balloon:Exec
    MountedTurret:Exec
    CreativeTool:Exec
    ExplosiveBow:Exec
    AshtonIndigo:Exec
    AshtonChicago:Exec
    MeleeDualWield:Exec
    Unarmed:Exec
   )
   Branch(execute=[
     @n5.AR_BullPup
     @n5.AR_ForwardGrip
     @n5.AssaultRifle
     @n5.Crossbow
     @n5.GrenadeLauncher
     @n5.Rifle
     @n5.RocketLauncher
     @n5.Shotgun
     @n5.SniperRifle
     @n5.TacticalShotgun
    ])
   > Set Rifle(Rifle=true)
   n6 = Branch(execute=@n1.then_3, Condition=Get Rifle() -> (Rifle:bool))
   > Set Rifle_alpha(Rifle_alpha:double=1) -> (Output_Get:double)
   Set Rifle_alpha(execute=@n6.else, Rifle_alpha=0)
   Set ADS(execute=@n1.then_4, ADS:double=SelectFloat(A=0.1, B=1, bPickA=Get bIsTargeting())) -> (Output_Get:double)
  ]}
]
variables [
 ADS {type=number}
 body_mesh {type=SkeletalMeshComponent}
 face_mesh {type=SkeletalMeshComponent}
 Gravity Override Params {type=GravityOverrideParamsStruct}
 GravityOverride {type=Vector}
 Rifle {type=boolean}
 Rifle_alpha {type=number}
]
